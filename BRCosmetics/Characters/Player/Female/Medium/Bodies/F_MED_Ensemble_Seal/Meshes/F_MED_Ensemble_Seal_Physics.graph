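# Physics asset for the female medium Seal ensemble cosmetic, rigging the hoodie, jacket flaps, shorts and strings for secondary motion.
# Capsules cover the pelvis, spine, clavicles and legs, with small master bodies linking each garment group to them through angularly free constraints.
# Jacket and shorts pieces hinge on their masters with one limited swing axis.
# The hood branches with swing and twist limits, while string balls form two twist-locked, three-link chains.
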PhysicsAsset /BRCosmetics/Characters/Player/Female/Medium/Bodies/F_MED_Ensemble_Seal/Meshes/F_MED_Ensemble_Seal_Physics
bodies {
 Dyn_Hoodie1 [{shapeType=Capsule, length=5, radius=5, shapeName=Dyn_Hoodie1_capsule, center=<2 0 0>}]
 Dyn_Hoodie2 [{shapeType=Capsule, length=5, radius=5, shapeName=Dyn_Hoodie2_capsule, center=<2 -2.7661 0>}]
 Dyn_Hoodie3 [{shapeType=Capsule, length=5, radius=5, shapeName=Dyn_Hoodie3_capsule, center=<2 -2.7661 0>}]
 Dyn_Master_Hoodie [{shapeType=Capsule, length=0.505, radius=0.505, shapeName=Dyn_Master_Hoodie_capsule, rotation=<rot 0 -90 0>}]
 Dyn_Master_Jacket [{shapeType=Capsule, length=0.5, radius=0.5, shapeName=Dyn_Master_Jacket_capsule}]
 Dyn_Master_Jacket_BL [{shapeType=Capsule, length=5, radius=7, shapeName=Dyn_Master_Jacket_BL_capsule, center=<3.4443 -6.9696 0>}]
 Dyn_Master_Jacket_BM [{shapeType=Capsule, length=5, radius=7, shapeName=Dyn_Master_Jacket_BM_capsule, center=<3.4443 -6.9696 0>}]
 Dyn_Master_Jacket_BR [{shapeType=Capsule, length=5, radius=7, shapeName=Dyn_Master_Jacket_BR_capsule, center=<3.4443 -6.9696 0>}]
 Dyn_Master_Jacket_FL [{shapeType=Capsule, length=5, radius=7, shapeName=Dyn_Master_Jacket_FL_capsule, center=<3.4443 -3.6395 0>}]
 Dyn_Master_Jacket_FR [{shapeType=Capsule, length=5, radius=7, shapeName=Dyn_Master_Jacket_FR_capsule, center=<3.4443 -3.6395 0>}]
 Dyn_Master_Jacket_L [{shapeType=Capsule, length=10, radius=7, shapeName=Dyn_Master_Jacket_L_capsule, center=<3.4443 -7.2853 0>}]
 Dyn_Master_Jacket_R [{shapeType=Capsule, length=10, radius=7, shapeName=Dyn_Master_Jacket_R_capsule, center=<3.4443 -7.2853 0>}]
 Dyn_Master_Short_L [{shapeType=Capsule, length=0.5, radius=0.5, shapeName=Dyn_Master_Short_L_capsule}]
 Dyn_Master_Short_R [{shapeType=Capsule, length=0.5, radius=0.5, shapeName=Dyn_Master_Short_R_capsule}]
 Dyn_Master_Short_R1 [{shapeType=Capsule, length=9, radius=5, shapeName=Dyn_Master_Short_R1_capsule, center=<5.1643 -5.8 0>}]
 Dyn_Master_Short_R2 [{shapeType=Capsule, length=9, radius=5, shapeName=Dyn_Master_Short_R2_capsule, center=<5.1643 -4.7915 0>}]
 Dyn_Master_Short_R3 [{shapeType=Capsule, length=9, radius=5, shapeName=Dyn_Master_Short_R3_capsule, center=<5.1643 -4.2249 0>}]
 Dyn_Master_Short_R4 [{shapeType=Capsule, length=9, radius=5, shapeName=Dyn_Master_Short_R4_capsule, center=<5.1643 -2.3641 0>}]
 Dyn_Master_String1 [{shapeType=Capsule, length=0.5, radius=0.5, shapeName=Dyn_Master_String1_capsule}]
 Dyn_Master_String2 [{shapeType=Capsule, length=0.5, radius=0.5, shapeName=Dyn_Master_String2_capsule}]
 Dyn_Short_L1 [{shapeType=Capsule, length=9, radius=5, shapeName=Dyn_Short_L1_capsule, center=<5.3789 -5.8 0>}]
 Dyn_Short_L2 [{shapeType=Capsule, length=9, radius=5, shapeName=Dyn_Short_L2_capsule, center=<5.3789 -3.7794 0>}]
 Dyn_Short_L3 [{shapeType=Capsule, length=9, radius=5, shapeName=Dyn_Short_L3_capsule, center=<5.3789 2.9961 0>}]
 Dyn_Short_L4 [{shapeType=Capsule, length=9, radius=5, shapeName=Dyn_Short_L4_capsule, center=<5.3789 -2.791 0>}]
 Dyn_String_Ball_L1 [{shapeType=Capsule, length=5, radius=1.5, shapeName=Dyn_String_Ball_L1_capsule, center=<1.8399 0 0>, rotation=<rot 90 0 0>}]
 Dyn_String_Ball_L2 [{shapeType=Capsule, length=5, radius=1.5, shapeName=Dyn_String_Ball_L2_capsule, center=<1.8399 0 0>, rotation=<rot 90 0 0>}]
 Dyn_String_Ball_L3 [{shapeType=Capsule, length=5, radius=1.5, shapeName=Dyn_String_Ball_L3_capsule, center=<1.8399 0 0>, rotation=<rot 90 0 0>}]
 Dyn_String_Ball_R1 [{shapeType=Capsule, length=5, radius=1.5, shapeName=Dyn_String_Ball_R1_capsule, center=<1.8399 0 0>, rotation=<rot 90 0 0>}]
 Dyn_String_Ball_R2 [{shapeType=Capsule, length=5, radius=1.5, shapeName=Dyn_String_Ball_R2_capsule, center=<1.8399 0 0>, rotation=<rot 90 0 0>}]
 Dyn_String_Ball_R3 [{shapeType=Capsule, length=5, radius=1.5, shapeName=Dyn_String_Ball_R3_capsule, center=<1.8399 0 0>, rotation=<rot 90 0 0>}]
 Dyn_String_Loop1 [{shapeType=Capsule, length=4, radius=3, shapeName=Dyn_String_Loop1_capsule, center=<2.9128 0.4008 1.6237>, rotation=<rot 20 0.0002 0.0001>}]
 calf_l [{shapeType=Capsule, length=22.7108, radius=6.2799, shapeName=calf_l_capsule, center=<-14.1847 -1.2616 0>, rotation=<rot 90 0 0>}]
 calf_r [{shapeType=Capsule, length=23.3658, radius=6.9637, shapeName=calf_r_capsule, center=<14.9956 0.4921 0>, rotation=<rot 90 0 0>}]
 clavicle_l [{shapeType=Capsule, length=5, radius=7, shapeName=clavicle_l_capsule, center=<7.7676 0 -2.6559>, rotation=<rot 90 0 0>}]
 clavicle_r [{shapeType=Capsule, length=5, radius=7, shapeName=clavicle_r_capsule, center=<-7.3228 -0.2834 2.2681>, rotation=<rot 90 0 0>}]
 pelvis [{shapeType=Capsule, length=9, radius=11.25, shapeName=pelvis_capsule}]
 spine_03 [{shapeType=Capsule, length=10, radius=8, shapeName=spine_03_capsule, center=<2.8294 5.7667 0>}]
 spine_05 [{shapeType=Capsule, length=5, radius=9.75, shapeName=spine_05_capsule, center=<0 0.5784 0>}]
 thigh_l [{shapeType=Capsule, length=20, radius=6, shapeName=thigh_l_capsule, center=<-26.5289 0 0.1485>, rotation=<rot 90 0 0>}]
 thigh_r [{shapeType=Capsule, length=20, radius=6, shapeName=thigh_r_capsule, center=<26.698 0 0>, rotation=<rot 90 0 0>}]
}
constraints [
 {
  bone1Name=Dyn_Master_Jacket
  bone2Name=pelvis
  swing1LimitDegrees=45
  swing1Motion=Free
  swing2LimitDegrees=45
  swing2Motion=Free
  twistLimitDegrees=45
  twistMotion=Free
 }
 {
  bone1Name=Dyn_Master_Jacket_FR
  bone2Name=Dyn_Master_Jacket
  swing1LimitDegrees=45
  swing2LimitDegrees=45
  swing2Motion=Locked
  twistLimitDegrees=45
  twistMotion=Locked
 }
 {
  bone1Name=Dyn_Master_Jacket_FL
  bone2Name=Dyn_Master_Jacket
  swing1LimitDegrees=45
  swing2LimitDegrees=45
  swing2Motion=Locked
  twistLimitDegrees=45
  twistMotion=Locked
 }
 {
  bone1Name=Dyn_Master_Jacket_R
  bone2Name=Dyn_Master_Jacket
  swing1LimitDegrees=45
  swing2LimitDegrees=45
  swing2Motion=Locked
  twistLimitDegrees=45
  twistMotion=Locked
 }
 {
  bone1Name=Dyn_Master_Jacket_L
  bone2Name=Dyn_Master_Jacket
  swing1LimitDegrees=45
  swing2LimitDegrees=45
  swing2Motion=Locked
  twistLimitDegrees=45
  twistMotion=Locked
 }
 {
  bone1Name=Dyn_Master_Jacket_BM
  bone2Name=Dyn_Master_Jacket
  swing1LimitDegrees=45
  swing2LimitDegrees=45
  swing2Motion=Locked
  twistLimitDegrees=45
  twistMotion=Locked
 }
 {
  bone1Name=Dyn_Master_Jacket_BL
  bone2Name=Dyn_Master_Jacket
  swing1LimitDegrees=45
  swing2LimitDegrees=45
  swing2Motion=Locked
  twistLimitDegrees=45
  twistMotion=Locked
 }
 {
  bone1Name=Dyn_Master_Jacket_BR
  bone2Name=Dyn_Master_Jacket
  swing1LimitDegrees=45
  swing2LimitDegrees=45
  swing2Motion=Locked
  twistLimitDegrees=45
  twistMotion=Locked
 }
 {
  bone1Name=Dyn_Master_Hoodie
  bone2Name=spine_05
  swing1LimitDegrees=45
  swing1Motion=Free
  swing2LimitDegrees=45
  swing2Motion=Free
  twistLimitDegrees=45
  twistMotion=Free
 }
 {bone1Name=Dyn_Hoodie1, bone2Name=Dyn_Master_Hoodie, swing1LimitDegrees=40, swing2LimitDegrees=20, twistLimitDegrees=5}
 {bone1Name=Dyn_Hoodie2, bone2Name=Dyn_Hoodie1, swing1LimitDegrees=40, swing2LimitDegrees=20, twistLimitDegrees=5}
 {bone1Name=Dyn_Hoodie3, bone2Name=Dyn_Hoodie1, swing1LimitDegrees=40, swing2LimitDegrees=20, twistLimitDegrees=5}
 {
  bone1Name=spine_05
  bone2Name=spine_03
  swing1LimitDegrees=45
  swing1Motion=Free
  swing2LimitDegrees=45
  swing2Motion=Free
  twistLimitDegrees=45
  twistMotion=Free
 }
 {
  bone1Name=clavicle_l
  bone2Name=spine_05
  swing1LimitDegrees=45
  swing1Motion=Free
  swing2LimitDegrees=45
  swing2Motion=Free
  twistLimitDegrees=45
  twistMotion=Free
 }
 {
  bone1Name=clavicle_r
  bone2Name=spine_05
  swing1LimitDegrees=45
  swing1Motion=Free
  swing2LimitDegrees=45
  swing2Motion=Free
  twistLimitDegrees=45
  twistMotion=Free
 }
 {
  bone1Name=spine_03
  bone2Name=pelvis
  swing1LimitDegrees=45
  swing1Motion=Free
  swing2LimitDegrees=45
  swing2Motion=Free
  twistLimitDegrees=45
  twistMotion=Free
 }
 {
  bone1Name=Dyn_Master_Short_L
  bone2Name=thigh_l
  swing1LimitDegrees=45
  swing1Motion=Free
  swing2LimitDegrees=45
  swing2Motion=Free
  twistLimitDegrees=45
  twistMotion=Free
 }
 {
  bone1Name=Dyn_Master_Short_R
  bone2Name=thigh_r
  swing1LimitDegrees=45
  swing1Motion=Free
  swing2LimitDegrees=45
  swing2Motion=Free
  twistLimitDegrees=45
  twistMotion=Free
 }
 {
  bone1Name=thigh_r
  bone2Name=pelvis
  swing1LimitDegrees=45
  swing1Motion=Free
  swing2LimitDegrees=45
  swing2Motion=Free
  twistLimitDegrees=45
  twistMotion=Free
 }
 {
  bone1Name=thigh_l
  bone2Name=pelvis
  swing1LimitDegrees=45
  swing1Motion=Free
  swing2LimitDegrees=45
  swing2Motion=Free
  twistLimitDegrees=45
  twistMotion=Free
 }
 {
  bone1Name=Dyn_Master_Short_R1
  bone2Name=Dyn_Master_Short_R
  swing1LimitDegrees=20
  swing2LimitDegrees=0
  swing2Motion=Locked
  twistLimitDegrees=0
  twistMotion=Locked
 }
 {
  bone1Name=Dyn_Master_Short_R2
  bone2Name=Dyn_Master_Short_R
  swing1LimitDegrees=30
  swing2LimitDegrees=45
  swing2Motion=Locked
  twistLimitDegrees=45
  twistMotion=Locked
 }
 {
  bone1Name=Dyn_Master_Short_R3
  bone2Name=Dyn_Master_Short_R
  swing1LimitDegrees=10
  swing2LimitDegrees=45
  swing2Motion=Locked
  twistLimitDegrees=45
  twistMotion=Locked
 }
 {
  bone1Name=Dyn_Master_Short_R4
  bone2Name=Dyn_Master_Short_R
  swing1LimitDegrees=10
  swing2LimitDegrees=45
  swing2Motion=Locked
  twistLimitDegrees=45
  twistMotion=Locked
 }
 {
  bone1Name=calf_r
  bone2Name=thigh_r
  swing1LimitDegrees=45
  swing1Motion=Free
  swing2LimitDegrees=45
  swing2Motion=Free
  twistLimitDegrees=45
  twistMotion=Free
 }
 {
  bone1Name=calf_l
  bone2Name=thigh_l
  swing1LimitDegrees=45
  swing1Motion=Free
  swing2LimitDegrees=45
  swing2Motion=Free
  twistLimitDegrees=45
  twistMotion=Free
 }
 {
  bone1Name=Dyn_Short_L4
  bone2Name=Dyn_Master_Short_L
  swing1LimitDegrees=10
  swing2LimitDegrees=45
  swing2Motion=Locked
  twistLimitDegrees=45
  twistMotion=Locked
 }
 {
  bone1Name=Dyn_Short_L3
  bone2Name=Dyn_Master_Short_L
  swing1LimitDegrees=10
  swing2LimitDegrees=45
  swing2Motion=Locked
  twistLimitDegrees=45
  twistMotion=Locked
 }
 {
  bone1Name=Dyn_Short_L2
  bone2Name=Dyn_Master_Short_L
  swing1LimitDegrees=30
  swing2LimitDegrees=45
  swing2Motion=Locked
  twistLimitDegrees=45
  twistMotion=Locked
 }
 {
  bone1Name=Dyn_Short_L1
  bone2Name=Dyn_Master_Short_L
  swing1LimitDegrees=20
  swing2LimitDegrees=45
  swing2Motion=Locked
  twistLimitDegrees=45
  twistMotion=Locked
 }
 {
  bone1Name=Dyn_Master_String1
  bone2Name=spine_05
  swing1LimitDegrees=45
  swing1Motion=Free
  swing2LimitDegrees=45
  swing2Motion=Free
  twistLimitDegrees=45
  twistMotion=Free
 }
 {
  bone1Name=Dyn_Master_String2
  bone2Name=spine_05
  swing1LimitDegrees=45
  swing1Motion=Free
  swing2LimitDegrees=45
  swing2Motion=Free
  twistLimitDegrees=45
  twistMotion=Free
 }
 {
  bone1Name=Dyn_String_Loop1
  bone2Name=Dyn_Master_String1
  swing1LimitDegrees=40
  swing2LimitDegrees=45
  swing2Motion=Locked
  twistLimitDegrees=45
  twistMotion=Locked
 }
 {
  bone1Name=Dyn_String_Ball_L1
  bone2Name=Dyn_Master_String2
  swing1LimitDegrees=30
  swing2LimitDegrees=20
  twistLimitDegrees=45
  twistMotion=Locked
 }
 {
  bone1Name=Dyn_String_Ball_L2
  bone2Name=Dyn_String_Ball_L1
  swing1LimitDegrees=30
  swing2LimitDegrees=20
  twistLimitDegrees=45
  twistMotion=Locked
 }
 {
  bone1Name=Dyn_String_Ball_L3
  bone2Name=Dyn_String_Ball_L2
  swing1LimitDegrees=30
  swing2LimitDegrees=20
  twistLimitDegrees=45
  twistMotion=Locked
 }
 {
  bone1Name=Dyn_String_Ball_R1
  bone2Name=Dyn_Master_String1
  swing1LimitDegrees=30
  swing2LimitDegrees=20
  twistLimitDegrees=45
  twistMotion=Locked
 }
 {
  bone1Name=Dyn_String_Ball_R2
  bone2Name=Dyn_String_Ball_R1
  swing1LimitDegrees=30
  swing2LimitDegrees=20
  twistLimitDegrees=45
  twistMotion=Locked
 }
 {
  bone1Name=Dyn_String_Ball_R3
  bone2Name=Dyn_String_Ball_R2
  swing1LimitDegrees=30
  swing2LimitDegrees=20
  twistLimitDegrees=45
  twistMotion=Locked
 }
]
