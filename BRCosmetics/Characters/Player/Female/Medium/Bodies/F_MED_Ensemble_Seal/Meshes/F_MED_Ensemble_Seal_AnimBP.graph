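# Post-process-style AnimBP for the female medium Seal ensemble body cosmetic, adding sleeve correctives and secondary motion to a pose copied from the attached parent mesh.
# Update derives hand-versus-forearm-twist rotation alphas, which drive sleeve bone constraints and additive offsets layered over weighted base constraints.
# A pelvis-space RigidBody simulation follows, then crouch-gated offsets on Dyn_Short bones.
# An AnimDynamics node is unconnected, and both front-end branches set IsFrontEnd to 1.
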AnimBlueprint /BRCosmetics/Characters/Player/Female/Medium/Bodies/F_MED_Ensemble_Seal/Meshes/F_MED_Ensemble_Seal_AnimBP
target_skeleton=&/F_MED_Ensemble_Seal_Skeleton
graphs [
 AnimGraph {nodes [
   Anim Dynamics - Bone: Dyn_String_Loop1(Alpha=1, ExternalForce="0,0,0") {
    angularSpringConstant=50
    bUseGravityOverride=true
    bUseSphericalLimits=true
    linearSpringConstant=100
    simulationSpace=BoneRelative
    alphaBoolBlend {blendOption=Linear}
    boundBone {boneName=Dyn_String_Loop1}
    componentAppliedLinearAccClamp=<5000 5000 5000>
    componentLinearAccScale=<0.5 0.5 0.5>
    componentLinearVelScale=<0.5 0.5 0.5>
    gravityOverride {z=-980}
    physicsBodyDefinitions [{
      boundBone {boneName=Dyn_String_Loop1}
      boxExtents=<8 2 2>
      constraintSetup {coneAngle=5, angularLimitsMax=<0 0 10>, angularLimitsMin=<0 0 0>, angularTarget=<0 0 -10>}
      localJointOffset=<1.5 0 0>
     }]
    relativeSpaceBone {boneName=spine_05}
    sphericalLimits [{limitRadius=5, drivingBone {boneName=Dyn_Master_String1}, sphereLocalOffset=<3 -6.25 0>}]
   }
   "Transform (Modify) Bone - Bone: Dyn_Short_L1"!(ComponentPose, bAlphaBoolEnabled:bool=True, Translation="0,3,0", Rotation="-0.015,-30.000,0.027", Scale="1,1,1") -> (Pose) {
    alphaInputType=Bool
    rotationMode=BMM_Additive
    rotationSpace=BCS_BoneSpace
    scaleMode=BMM_Additive
    scaleSpace=BCS_BoneSpace
    translationMode=BMM_Additive
    translationSpace=BCS_BoneSpace
    alphaBoolBlend {blendOption=Linear}
    boneToModify {boneName=Dyn_Short_L1}
    rotation=<rot -0.015 -30.0001 0.027>
    translation {y=3}
   }
   n1 = Get bIsCrouching()
   n2 = "Transform (Modify) Bone - Bone: Dyn_Short_L2"!(ComponentPose, bAlphaBoolEnabled:bool=@n1, Translation="-9.958,-14.048,3.277", Rotation="-0.013,50.000,-0.002", Scale="1,1,1") -> (Pose) {
    alphaInputType=Bool
    rotationMode=BMM_Additive
    rotationSpace=BCS_BoneSpace
    scaleMode=BMM_Additive
    scaleSpace=BCS_BoneSpace
    translationMode=BMM_Additive
    translationSpace=BCS_BoneSpace
    alphaBoolBlend {blendOption=Linear}
    boneToModify {boneName=Dyn_Short_L2}
    rotation=<rot -0.013 50.0001 -0.002>
    translation=<-9.958 -14.048 3.277>
   }
   "Transform (Modify) Bone - Bone: Dyn_Master_Short_R2"!(ComponentPose=@n2, bAlphaBoolEnabled:bool=@n1, Translation="-4.671,-14.605,-5.826", Rotation="0,0,0", Scale="1,1,1") -> (Pose) {
    alphaInputType=Bool
    rotationMode=BMM_Additive
    scaleMode=BMM_Additive
    scaleSpace=BCS_BoneSpace
    translationMode=BMM_Additive
    translationSpace=BCS_BoneSpace
    alphaBoolBlend {blendOption=Linear}
    boneToModify {boneName=Dyn_Master_Short_R2}
    translation=<-4.671 -14.605 -5.826>
   }
   n3 = "Constraint - hand_r_Sreeves_01 (PP)"!(
    ComponentPose=Local To Component(LocalPose=Copy Pose From Mesh() {bUseAttachedParent=true})
    Alpha=1
    ConstraintWeights_0=0.2
    ConstraintWeights_1:float=0.8
   ) -> (Pose) {
    alphaBoolBlend {blendOption=Linear}
    boneToModify {boneName=hand_r_Sreeves_01}
    constraintSetup [{targetBone {boneName=hand_r}}, {targetBone {boneName=lowerarm_twist_01_r}}]
    constraintWeights [1,1]
   }
   n4 = "Constraint - hand_r_Sreeves_03 (PP)"!(ComponentPose=@n3, Alpha=1, ConstraintWeights_0=0.2, ConstraintWeights_1:float=0.8) -> (Pose) {
    alphaBoolBlend {blendOption=Linear}
    boneToModify {boneName=hand_r_Sreeves_03}
    constraintSetup [{targetBone {boneName=hand_r}}, {targetBone {boneName=lowerarm_twist_01_r}}]
    constraintWeights [1,1]
   }
   n5 = "Constraint - hand_r_Sreeves_04 (PP)"!(ComponentPose=@n4, Alpha=1, ConstraintWeights_0=0.2, ConstraintWeights_1:float=0.8) -> (Pose) {
    alphaBoolBlend {blendOption=Linear}
    boneToModify {boneName=hand_r_Sreeves_04}
    constraintSetup [{targetBone {boneName=hand_r}}, {targetBone {boneName=lowerarm_twist_01_r}}]
    constraintWeights [1,1]
   }
   n6 = "Constraint - hand_l_Sleeves_01 (PP)"!(ComponentPose=@n5, Alpha=1, ConstraintWeights_0=0.2, ConstraintWeights_1:float=0.8) -> (Pose) {
    alphaBoolBlend {blendOption=Linear}
    boneToModify {boneName=hand_l_Sleeves_01}
    constraintSetup [{targetBone {boneName=hand_l}}, {targetBone {boneName=lowerarm_twist_01_l}}]
    constraintWeights [1,1]
   }
   n7 = "Constraint - hand_l_Sleeves_03 (PP)"!(ComponentPose=@n6, Alpha=1, ConstraintWeights_0=0.2, ConstraintWeights_1:float=0.8) -> (Pose) {
    alphaBoolBlend {blendOption=Linear}
    boneToModify {boneName=hand_l_Sleeves_03}
    constraintSetup [{targetBone {boneName=hand_l}}, {targetBone {boneName=lowerarm_twist_01_l}}]
    constraintWeights [1,1]
   }
   n8 = "Constraint - hand_l_Sleeves_04 (PP)"!(ComponentPose=@n7, Alpha=1, ConstraintWeights_0=0.2, ConstraintWeights_1:float=0.8) -> (Pose) {
    alphaBoolBlend {blendOption=Linear}
    boneToModify {boneName=hand_l_Sleeves_04}
    constraintSetup [{targetBone {boneName=hand_l}}, {targetBone {boneName=lowerarm_twist_01_l}}]
    constraintWeights [1,1]
   }
   n9 = Get hand_shape_up_r() -> (hand_shape_up_r:double)
   n10 = "Constraint - hand_r_Sreeves_01 (PP)"(ComponentPose=@n8, Alpha=@n9, ConstraintWeights_0=1, ConstraintWeights_1=0) {
    alphaBoolBlend {blendOption=Linear}
    boneToModify {boneName=hand_r_Sreeves_01}
    constraintSetup [{targetBone {boneName=hand_r}}, {targetBone {boneName=lowerarm_twist_01_r}}]
    constraintWeights [1,1]
   }
   n11 = "Transform (Modify) Bone - Bone: hand_r_Sreeves_01"(ComponentPose=@n10, Alpha=@n9, Translation="1,-1,0", Rotation="0,0,0", Scale="1,1,1") {
    rotationMode=BMM_Additive
    rotationSpace=BCS_BoneSpace
    scaleMode=BMM_Additive
    scaleSpace=BCS_BoneSpace
    translationMode=BMM_Additive
    translationSpace=BCS_BoneSpace
    alphaBoolBlend {blendOption=Linear}
    boneToModify {boneName=hand_r_Sreeves_01}
    translation=<1 -1>
   }
   n12 = Get hand_shape_dn_r() -> (hand_shape_dn_r:double)
   n13 = "Constraint - hand_r_Sreeves_03 (PP)"(ComponentPose=@n11, Alpha=@n12, ConstraintWeights_0=0.2, ConstraintWeights_1=0.8) {
    alphaBoolBlend {blendOption=Linear}
    boneToModify {boneName=hand_r_Sreeves_03}
    constraintSetup [{targetBone {boneName=hand_r}}, {targetBone {boneName=lowerarm_twist_01_r}}]
    constraintWeights [1,1]
   }
   n14 = "Transform (Modify) Bone - Bone: hand_r_Sreeves_03"(ComponentPose=@n13, Alpha=@n12, Translation="12,0,0", Rotation="0,0,0", Scale="1,1,1") {
    rotationMode=BMM_Additive
    rotationSpace=BCS_BoneSpace
    scaleMode=BMM_Additive
    scaleSpace=BCS_BoneSpace
    translationMode=BMM_Additive
    translationSpace=BCS_BoneSpace
    alphaBoolBlend {blendOption=Linear}
    boneToModify {boneName=hand_r_Sreeves_03}
    translation {x=12}
   }
   n15 = Get hand_shape_lt_r() -> (hand_shape_lt_r:double)
   n16 = "Constraint - hand_r_Sreeves_02 (PP)"!(ComponentPose=@n14, Alpha=@n15, ConstraintWeights_0=0.2, ConstraintWeights_1:float=0.8) -> (Pose) {
    alphaBoolBlend {blendOption=Linear}
    boneToModify {boneName=hand_r_Sreeves_02}
    constraintSetup [{targetBone {boneName=hand_r}}, {targetBone {boneName=lowerarm_twist_01_r}}]
    constraintWeights [1,1]
   }
   n17 = "Transform (Modify) Bone - Bone: hand_r_Sreeves_02"(ComponentPose=@n16, Alpha=@n15, Translation="12,0,0", Rotation="-20,0,0", Scale="1,1,1") {
    rotationMode=BMM_Additive
    rotationSpace=BCS_BoneSpace
    scaleMode=BMM_Additive
    scaleSpace=BCS_BoneSpace
    translationMode=BMM_Additive
    translationSpace=BCS_BoneSpace
    alphaBoolBlend {blendOption=Linear}
    boneToModify {boneName=hand_r_Sreeves_02}
    rotation {pitch=-20}
    translation {x=12}
   }
   n18 = Get hand_shape_rt_r() -> (hand_shape_rt_r:double)
   n19 = "Constraint - hand_r_Sreeves_04 (PP)"(ComponentPose=@n17, Alpha=@n18, ConstraintWeights_0=1, ConstraintWeights_1=0) {
    alphaBoolBlend {blendOption=Linear}
    boneToModify {boneName=hand_r_Sreeves_04}
    constraintSetup [{targetBone {boneName=hand_r}}, {targetBone {boneName=lowerarm_twist_01_r}}]
    constraintWeights [1,1]
   }
   n20 = "Transform (Modify) Bone - Bone: hand_r_Sreeves_04"(ComponentPose=@n19, Alpha=@n18, Translation="3,0,-2.3", Rotation="0,0,0", Scale="1,1,1") {
    rotationMode=BMM_Additive
    rotationSpace=BCS_BoneSpace
    scaleMode=BMM_Additive
    scaleSpace=BCS_BoneSpace
    translationMode=BMM_Additive
    translationSpace=BCS_BoneSpace
    alphaBoolBlend {blendOption=Linear}
    boneToModify {boneName=hand_r_Sreeves_04}
    translation {x=3, z=-2.3}
   }
   n21 = Get hand_shape_up_l() -> (hand_shape_up_l:double)
   n22 = "Constraint - hand_l_Sleeves_01 (PP)"(ComponentPose=@n20, Alpha=@n21, ConstraintWeights_0=1, ConstraintWeights_1=0) {
    alphaBoolBlend {blendOption=Linear}
    boneToModify {boneName=hand_l_Sleeves_01}
    constraintSetup [{targetBone {boneName=hand_l}}, {targetBone {boneName=lowerarm_twist_01_l}}]
    constraintWeights [1,1]
   }
   n23 = "Transform (Modify) Bone - Bone: hand_l_Sleeves_01"(ComponentPose=@n22, Alpha=@n21, Translation="-1,1,0", Rotation="0,0,0", Scale="1,1,1") {
    rotationMode=BMM_Additive
    rotationSpace=BCS_BoneSpace
    scaleMode=BMM_Additive
    scaleSpace=BCS_BoneSpace
    translationMode=BMM_Additive
    translationSpace=BCS_BoneSpace
    alphaBoolBlend {blendOption=Linear}
    boneToModify {boneName=hand_l_Sleeves_01}
    translation=<-1 1>
   }
   n24 = Get hand_shape_dn_l() -> (hand_shape_dn_l:double)
   n25 = "Constraint - hand_l_Sleeves_03 (PP)"(ComponentPose=@n23, Alpha=@n24, ConstraintWeights_0=0.2, ConstraintWeights_1=0.8) {
    alphaBoolBlend {blendOption=Linear}
    boneToModify {boneName=hand_l_Sleeves_03}
    constraintSetup [{targetBone {boneName=hand_l}}, {targetBone {boneName=lowerarm_twist_01_l}}]
    constraintWeights [1,1]
   }
   n26 = "Transform (Modify) Bone - Bone: hand_l_Sleeves_03"(ComponentPose=@n25, Alpha=@n24, Translation="-12,0,0", Rotation="0,0,0", Scale="1,1,1") {
    rotationMode=BMM_Additive
    rotationSpace=BCS_BoneSpace
    scaleMode=BMM_Additive
    scaleSpace=BCS_BoneSpace
    translationMode=BMM_Additive
    translationSpace=BCS_BoneSpace
    alphaBoolBlend {blendOption=Linear}
    boneToModify {boneName=hand_l_Sleeves_03}
    translation {x=-12}
   }
   n27 = Get hand_shape_lt_l() -> (hand_shape_lt_l:double)
   n28 = "Constraint - hand_l_Sleeves_02 (PP)"!(ComponentPose=@n26, Alpha=@n27, ConstraintWeights_0=0.2, ConstraintWeights_1:float=0.8) -> (Pose) {
    alphaBoolBlend {blendOption=Linear}
    boneToModify {boneName=hand_l_Sleeves_02}
    constraintSetup [{targetBone {boneName=hand_l}}, {targetBone {boneName=lowerarm_twist_01_l}}]
    constraintWeights [1,1]
   }
   n29 = "Transform (Modify) Bone - Bone: hand_l_Sleeves_02"(ComponentPose=@n28, Alpha=@n27, Translation="-12,0,0", Rotation="-20,0,0", Scale="1,1,1") {
    rotationMode=BMM_Additive
    rotationSpace=BCS_BoneSpace
    scaleMode=BMM_Additive
    scaleSpace=BCS_BoneSpace
    translationMode=BMM_Additive
    translationSpace=BCS_BoneSpace
    alphaBoolBlend {blendOption=Linear}
    boneToModify {boneName=hand_l_Sleeves_02}
    rotation {pitch=-20}
    translation {x=-12}
   }
   n30 = Get hand_shape_rt_l() -> (hand_shape_rt_l:double)
   n31 = "Constraint - hand_l_Sleeves_04 (PP)"(ComponentPose=@n29, Alpha=@n30, ConstraintWeights_0=1, ConstraintWeights_1=0) {
    alphaBoolBlend {blendOption=Linear}
    boneToModify {boneName=hand_l_Sleeves_04}
    constraintSetup [{targetBone {boneName=hand_l}}, {targetBone {boneName=lowerarm_twist_01_l}}]
    constraintWeights [1,1]
   }
   n32 = "Transform (Modify) Bone - Bone: hand_l_Sleeves_04"(ComponentPose=@n31, Alpha=@n30, Translation="-3,0,3", Rotation="-10,0,0", Scale="1,1,1") {
    rotationMode=BMM_Additive
    rotationSpace=BCS_BoneSpace
    scaleMode=BMM_Additive
    scaleSpace=BCS_BoneSpace
    translationMode=BMM_Additive
    translationSpace=BCS_BoneSpace
    alphaBoolBlend {blendOption=Linear}
    boneToModify {boneName=hand_l_Sleeves_04}
    rotation {pitch=-10}
    translation {x=-3, z=3}
   }
   n33 = RigidBody(ComponentPose=@n32) {
    bClampLinearTranslationLimitToRefPose=true
    bOverrideWorldGravity=true
    overridePhysicsAsset=&/F_MED_Ensemble_Seal_Physics
    simulationSpace=BaseBoneSpace
    alphaBoolBlend {blendOption=Linear}
    baseBoneRef {boneName=pelvis}
    componentAppliedLinearAccClamp=<1000 1000 1000>
    componentLinearAccScale=<0.5 0.5 0.5>
    componentLinearVelScale=<0.5 0.5 0.5>
    overrideWorldGravity {z=-980}
   }
   n34 = "Transform (Modify) Bone - Bone: Dyn_Short_L1"(ComponentPose=@n33, bAlphaBoolEnabled=@n1, Translation="0,1,0", Rotation="0,0,0", Scale="1,1,1") {
    alphaInputType=Bool
    rotationMode=BMM_Additive
    rotationSpace=BCS_BoneSpace
    scaleMode=BMM_Additive
    scaleSpace=BCS_BoneSpace
    translationMode=BMM_Additive
    translationSpace=BCS_BoneSpace
    alphaBoolBlend {blendOption=Linear}
    boneToModify {boneName=Dyn_Short_L1}
    translation {y=1}
   }
   n35 = "Transform (Modify) Bone - Bone: Dyn_Short_L2"(ComponentPose=@n34, bAlphaBoolEnabled=@n1, Translation="-14,0,0", Rotation="0,0,0", Scale="1,1,1") {
    alphaInputType=Bool
    rotationMode=BMM_Additive
    rotationSpace=BCS_BoneSpace
    scaleMode=BMM_Additive
    scaleSpace=BCS_BoneSpace
    translationMode=BMM_Additive
    translationSpace=BCS_BoneSpace
    alphaBoolBlend {blendOption=Linear}
    boneToModify {boneName=Dyn_Short_L2}
    translation {x=-14}
   }
   n36 = "Transform (Modify) Bone - Bone: Dyn_Master_Short_R2"(ComponentPose=@n35, bAlphaBoolEnabled=@n1, Translation="-18,0,0", Rotation="0,0,0", Scale="1,1,1") {
    alphaInputType=Bool
    rotationMode=BMM_Additive
    scaleMode=BMM_Additive
    scaleSpace=BCS_BoneSpace
    translationMode=BMM_Additive
    translationSpace=BCS_BoneSpace
    alphaBoolBlend {blendOption=Linear}
    boneToModify {boneName=Dyn_Master_Short_R2}
    translation {x=-18}
   }
   Output Pose(Result=Component To Local(ComponentPose=@n36)) {layerGroup=None}
  ]}
 EventGraph {nodes [
   Event BlueprintInitializeAnimation()
   > Set Skel_Mesh(Skel_Mesh=GetOwnerSkeletalMeshComponent())
   > n1 = Cast To PlayerPawn_Athena(Object=TryGetPawnOwner())
   > n2 = Set Pawn Owner(Pawn Owner:Fort Player Pawn Athena Object Reference=@n1) -> (Output_Get:Fort Player Pawn Athena Object Reference)
   > Set Shoulder_Mesh(Shoulder_Mesh:Skeletal Mesh Component Object Reference=GetSkeletalMeshForPartType(self=@n2)) -> (Output_Get:Skeletal Mesh Component Object Reference)
   Event BlueprintUpdateAnimation()
   > Set hand_shape_up_l(hand_shape_up_l:double=GetAlphaFromReferenceTransformDelta(
     SkeletalMeshComponent=Get Skel_Mesh()
     BoneName=hand_l
     BaseName=lowerarm_twist_01_l
     DeltaComponentType=EulerX
     InputRange=(X=20,Y=-20)
     TwistAxis="0,0,1"
    )) -> (Output_Get:double)
   > Set hand_shape_up_r(hand_shape_up_r:double=GetAlphaFromReferenceTransformDelta(
     SkeletalMeshComponent=Get Skel_Mesh()
     BoneName=hand_r
     BaseName=lowerarm_twist_01_r
     DeltaComponentType=EulerX
     InputRange=(X=20,Y=-20)
     TwistAxis="0,0,1"
    )) -> (Output_Get:double)
   > Set hand_shape_dn_l(hand_shape_dn_l:double=GetAlphaFromReferenceTransformDelta(
     SkeletalMeshComponent=Get Skel_Mesh()
     BoneName=hand_l
     BaseName=lowerarm_twist_01_l
     DeltaComponentType=EulerY
     InputRange=(X=45,Y=-20)
     OutputRange=(X=1,Y=0)
     TwistAxis="0,0,1"
    )) -> (Output_Get:double)
   > Set hand_shape_dn_r(hand_shape_dn_r:double=GetAlphaFromReferenceTransformDelta(
     SkeletalMeshComponent=Get Skel_Mesh()
     BoneName=hand_r
     BaseName=lowerarm_twist_01_r
     DeltaComponentType=EulerY
     InputRange=(X=45,Y=-20)
     OutputRange=(X=1,Y=0)
     TwistAxis="0,0,1"
    )) -> (Output_Get:double)
   > Set hand_shape_lt_l(hand_shape_lt_l:double=GetAlphaFromReferenceTransformDelta(
     SkeletalMeshComponent=Get Skel_Mesh()
     BoneName=hand_l
     BaseName=lowerarm_twist_01_l
     DeltaComponentType=EulerX
     InputRange=(X=-10,Y=45)
     TwistAxis="0,0,1"
    )) -> (Output_Get:double)
   > Set hand_shape_lt_r(hand_shape_lt_r:double=GetAlphaFromReferenceTransformDelta(
     SkeletalMeshComponent=Get Skel_Mesh()
     BoneName=hand_r
     BaseName=lowerarm_twist_01_r
     DeltaComponentType=EulerX
     InputRange=(X=-10,Y=45)
     TwistAxis="0,0,1"
    )) -> (Output_Get:double)
   > Set hand_shape_rt_l(hand_shape_rt_l:double=GetAlphaFromReferenceTransformDelta(
     SkeletalMeshComponent=Get Skel_Mesh()
     BoneName=hand_l
     BaseName=lowerarm_twist_01_l
     DeltaComponentType=EulerX
     InputRange=(X=45,Y=-20)
     TwistAxis="0,0,1"
    )) -> (Output_Get:double)
   > Set hand_shape_rt_r(hand_shape_rt_r:double=GetAlphaFromReferenceTransformDelta(
     SkeletalMeshComponent=Get Skel_Mesh()
     BoneName=hand_r
     BaseName=lowerarm_twist_01_r
     DeltaComponentType=EulerX
     InputRange=(X=45,Y=-20)
     TwistAxis="0,0,1"
    )) -> (Output_Get:double)
   > n3 = Branch(Condition=Get bIsInFrontEnd())
   > Set IsFrontEnd(IsFrontEnd:double=1) -> (Output_Get:double)
   Set IsFrontEnd(execute=@n3.else, IsFrontEnd=1)
  ]}
]
variables [
 hand_shape_dn_l {type=number}
 hand_shape_dn_r {type=number}
 hand_shape_lt_l {type=number}
 hand_shape_lt_r {type=number}
 hand_shape_rt_l {type=number}
 hand_shape_rt_r {type=number}
 hand_shape_up_l {type=number}
 hand_shape_up_r {type=number}
 IsFrontEnd {type=number}
 Pawn Owner {type=FortPlayerPawnAthena}
 Shoulder_Mesh {type=SkeletalMeshComponent}
 Skel_Mesh {type=SkeletalMeshComponent}
]
