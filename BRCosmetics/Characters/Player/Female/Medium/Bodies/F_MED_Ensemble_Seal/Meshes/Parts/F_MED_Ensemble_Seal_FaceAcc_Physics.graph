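# Physics asset for the face accessory part of the medium female player's Seal ensemble, defining capsule bodies on dynamic hair bones.
# Left, right and front master anchors link to a central hair master with free rotation.
# Three-segment back, left and right chains and a single front segment hang from these masters.
# Strand joints lock twist and limit swing, with wider limits down the back chain and on the front.
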PhysicsAsset /BRCosmetics/Characters/Player/Female/Medium/Bodies/F_MED_Ensemble_Seal/Meshes/Parts/F_MED_Ensemble_Seal_FaceAcc_Physics
bodies {
 Dyn_HairBack1 [{shapeType=Capsule, length=3, radius=5, shapeName=Dyn_HairBack1_capsule, center=<4.6723 0 0>, rotation=<rot 90 0 0>}]
 Dyn_HairBack2 [{shapeType=Capsule, length=3, radius=5, shapeName=Dyn_HairBack2_capsule, center=<4.6723 0 0>, rotation=<rot 90 0 0>}]
 Dyn_HairBack3 [{shapeType=Capsule, length=1, radius=5, shapeName=Dyn_HairBack3_capsule, center=<3.3323 0 0>, rotation=<rot 90 0 0>}]
 Dyn_HairFront1 [{shapeType=Capsule, length=3, radius=5, shapeName=Dyn_HairFront1_capsule, center=<4.6723 0 0>, rotation=<rot 90 0 0>}]
 Dyn_HairLeft1 [{shapeType=Capsule, length=3, radius=5, shapeName=Dyn_HairLeft1_capsule, center=<4.6723 0 0>, rotation=<rot 90 0 0>}]
 Dyn_HairLeft2 [{shapeType=Capsule, length=3, radius=5, shapeName=Dyn_HairLeft2_capsule, center=<4.6723 0 0>, rotation=<rot 90 0 0>}]
 Dyn_HairLeft3 [{shapeType=Capsule, length=1, radius=5, shapeName=Dyn_HairLeft3_capsule, center=<3.3323 0 0>, rotation=<rot 90 0 0>}]
 Dyn_HairRight1 [{shapeType=Capsule, length=3, radius=5, shapeName=Dyn_HairRight1_capsule, center=<4.6723 0 0>, rotation=<rot 90 0 0>}]
 Dyn_HairRight2 [{shapeType=Capsule, length=3, radius=5, shapeName=Dyn_HairRight2_capsule, center=<4.6723 0 0>, rotation=<rot 90 0 0>}]
 Dyn_HairRight3 [{shapeType=Capsule, length=1, radius=5, shapeName=Dyn_HairRight3_capsule, center=<3.3323 0 0>, rotation=<rot 90 0 0>}]
 Dyn_Master_Hair [{shapeType=Capsule, length=0.5, radius=0.5, shapeName=Dyn_Master_Hair_capsule}]
 Dyn_Master_HairFront [{shapeType=Capsule, length=0.5, radius=0.5, shapeName=Dyn_Master_HairFront_capsule}]
 Dyn_Master_HairLeft [{shapeType=Capsule, length=0.5, radius=0.5, shapeName=Dyn_Master_HairLeft_capsule}]
 Dyn_Master_HairRight [{shapeType=Capsule, length=0.5, radius=0.5, shapeName=Dyn_Master_HairRight_capsule}]
}
constraints [
 {
  bone1Name=Dyn_Master_HairLeft
  bone2Name=Dyn_Master_Hair
  swing1LimitDegrees=45
  swing1Motion=Free
  swing2LimitDegrees=45
  swing2Motion=Free
  twistLimitDegrees=45
  twistMotion=Free
 }
 {
  bone1Name=Dyn_Master_HairRight
  bone2Name=Dyn_Master_Hair
  swing1LimitDegrees=45
  swing1Motion=Free
  swing2LimitDegrees=45
  swing2Motion=Free
  twistLimitDegrees=45
  twistMotion=Free
 }
 {
  bone1Name=Dyn_Master_HairFront
  bone2Name=Dyn_Master_Hair
  swing1LimitDegrees=45
  swing1Motion=Free
  swing2LimitDegrees=45
  swing2Motion=Free
  twistLimitDegrees=45
  twistMotion=Free
 }
 {bone1Name=Dyn_HairBack1, bone2Name=Dyn_Master_Hair, swing1LimitDegrees=20, swing2LimitDegrees=10, twistLimitDegrees=45, twistMotion=Locked}
 {bone1Name=Dyn_HairBack2, bone2Name=Dyn_HairBack1, swing1LimitDegrees=30, swing2LimitDegrees=20, twistLimitDegrees=45, twistMotion=Locked}
 {bone1Name=Dyn_HairBack3, bone2Name=Dyn_HairBack2, swing1LimitDegrees=30, swing2LimitDegrees=20, twistLimitDegrees=45, twistMotion=Locked}
 {
  bone1Name=Dyn_HairLeft1
  bone2Name=Dyn_Master_HairLeft
  swing1LimitDegrees=20
  swing2LimitDegrees=10
  twistLimitDegrees=45
  twistMotion=Locked
 }
 {bone1Name=Dyn_HairLeft2, bone2Name=Dyn_HairLeft1, swing1LimitDegrees=20, swing2LimitDegrees=10, twistLimitDegrees=45, twistMotion=Locked}
 {bone1Name=Dyn_HairLeft3, bone2Name=Dyn_HairLeft2, swing1LimitDegrees=20, swing2LimitDegrees=10, twistLimitDegrees=45, twistMotion=Locked}
 {
  bone1Name=Dyn_HairRight1
  bone2Name=Dyn_Master_HairRight
  swing1LimitDegrees=15
  swing2LimitDegrees=10
  twistLimitDegrees=45
  twistMotion=Locked
 }
 {bone1Name=Dyn_HairRight2, bone2Name=Dyn_HairRight1, swing1LimitDegrees=20, swing2LimitDegrees=10, twistLimitDegrees=45, twistMotion=Locked}
 {bone1Name=Dyn_HairRight3, bone2Name=Dyn_HairRight2, swing1LimitDegrees=15, swing2LimitDegrees=10, twistLimitDegrees=45, twistMotion=Locked}
 {
  bone1Name=Dyn_HairFront1
  bone2Name=Dyn_Master_HairFront
  swing1LimitDegrees=30
  swing2LimitDegrees=20
  twistLimitDegrees=45
  twistMotion=Locked
 }
]
